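# Foley sound cue for the TwistEternity emote, using emote foley attenuation, sound class, property sheet and concurrency rules.
# A licensed content switcher root chooses among its child slots: a five-second test silence wave, an empty slot and a Creative radio accent stinger.
# No wave specific to the emote is referenced, only these generic assets.
# The cue also declares a send to the music audio-analysis submix.
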SoundCue /BRCosmetics/Sounds/Emotes/TwistEternity/EmoteFoley_TwistEternity_Cue
root=FortSoundNodeLicensedContentSwitcher_0
nodes [
 {id=FortSoundNodeLicensedContentSwitcher_0, class=FortSoundNodeLicensedContentSwitcher, children [WavePlayer_0,null,WavePlayer_1]}
 {id=WavePlayer_0, properties {soundWaveAssetPtr=/Game/Sounds/Fort_Test_Sounds/5_seconds_silence}}
 {id=WavePlayer_1, properties {soundWaveAssetPtr=/Game/Sounds/Creative/Gadgets/Radio/Stingers/Stinger_Accent_01}}
]
properties {
 attenuationSettings=&Fort_Emotes_Athena_Att
 audioPropertiesSheet=/BRCosmetics/Sounds/Settings/PropertySheets/Emotes/APS_Cosmetic_Emote_Foley
 maxDistance=1600
 soundClassObject=/Game/Sounds/FortSoundClassesAndModes/Fort_Foley_Emote
 volumeMultiplier=1
 concurrencyOverrides {maxCount=3}
 concurrencySet [/Game/Sounds/Concurrency_Rules/Athena_EmoteFoley_Conc]
 soundSubmixSends [{minSendLevel=0, soundSubmix=&Fortnite_AudioAnalysis_Music}]
}
